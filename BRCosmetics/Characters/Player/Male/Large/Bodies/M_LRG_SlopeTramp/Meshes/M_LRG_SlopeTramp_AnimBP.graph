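# Animation Blueprint for the large male SlopeTramp player body, layering cosmetic fixes over the pose copied from its parent mesh.
# Additive spine and clavicle offsets apply while targeting, then a Control Rig runs and constraints pull forearms to spline bones for hidden arms.
# The event graph feeds rig inputs from DBNO, weapon-name, skydive and emote checks.
# On customization load, it toggles tail visibility based on weapon and backbling.
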AnimBlueprint /BRCosmetics/Characters/Player/Male/Large/Bodies/M_LRG_SlopeTramp/Meshes/M_LRG_SlopeTramp_AnimBP
target_skeleton=&/M_LRG_SlopeTramp_Skeleton
graphs [
 AnimGraph {nodes [
   n1 = "Transform (Modify) Bone - Bone: spine_05"!(ComponentPose=Local To Component(LocalPose=Use cached pose 'TargetPose'()), Alpha=1, Translation="-2,2,0", Rotation="0,0,0") -> (Pose) {
    rotationMode=BMM_Additive
    translationMode=BMM_Additive
    alphaBoolBlend {blendOption=Linear}
    alphaScaleBiasClamp {bInterpResult=true, interpSpeedDecreasing=5, interpSpeedIncreasing=5}
    boneToModify {boneName=spine_05}
    translation=<-2 2>
   }
   n2 = "Transform (Modify) Bone - Bone: spine_04"!(ComponentPose=@n1, Alpha=1, Translation="-1,2,0", Rotation="0,0,0") -> (Pose) {
    rotationMode=BMM_Additive
    translationMode=BMM_Additive
    alphaBoolBlend {blendOption=Linear}
    alphaScaleBiasClamp {bInterpResult=true, interpSpeedDecreasing=5, interpSpeedIncreasing=5}
    boneToModify {boneName=spine_04}
    translation=<-1 2>
   }
   n3 = "Transform (Modify) Bone - Bone: clavicle_l"!(ComponentPose=@n2, Alpha=1, Translation="3,-4,0", Rotation="0,0,0") -> (Pose) {
    rotationMode=BMM_Additive
    translationMode=BMM_Additive
    alphaBoolBlend {blendOption=Linear}
    alphaScaleBiasClamp {bInterpResult=true, interpSpeedDecreasing=5, interpSpeedIncreasing=5}
    boneToModify {boneName=clavicle_l}
    translation=<3 -4>
   }
   n4 = "Transform (Modify) Bone - Bone: clavicle_r"!(ComponentPose=@n3, Alpha=1, Translation="3,-4,0", Rotation="0,0,0") -> (Pose) {
    rotationMode=BMM_Additive
    translationMode=BMM_Additive
    alphaBoolBlend {blendOption=Linear}
    alphaScaleBiasClamp {bInterpResult=true, interpSpeedDecreasing=5, interpSpeedIncreasing=5}
    boneToModify {boneName=clavicle_r}
    translation=<3 -4>
   }
   n5 = Control Rig!(
    Source=Blend Poses by bool(BlendPose_0=Component To Local(ComponentPose=@n4), BlendPose_1=Use cached pose 'TargetPose'(), bActiveValue=Get bIsTargeting()) {blendType=Linear}
    Alpha
    spr_upper_l_DBNO:double=Get spr_upper_l_DBNO() -> (spr_upper_l_DBNO:double)
    arms:double=Get arms() -> (arms:double)
    arms_0:double=Get arms_0() -> (arms_0:double)
    arms_1:double=Get arms_1() -> (arms_1:double)
    isFrontEndNoEmote:bool=Get isFrontEndNoEmote() -> (isFrontEndNoEmote:bool)
   ) -> (Pose) {
    alphaBoolBlend {blendOption=Linear}
    controlRigAssetReference {blueprintRigClass=&/M_LRG_SlopeTramp_CtrlRig.M_LRG_SlopeTramp_CtrlRig_C}
   }
   n6 = "Constraint - lowerarm_r (P)"(ComponentPose=Local To Component(LocalPose=@n5), Alpha=Get arms(), ConstraintWeights_0=1) {
    alphaBoolBlend {blendOption=Linear}
    boneToModify {boneName=lowerarm_r}
    constraintSetup [{targetBone {boneName=dfrm_arm_spline_09_r}}]
    constraintWeights [1]
   }
   n7 = "Constraint - lowerarm_l (P)"(ComponentPose=@n6, Alpha=Get arms_0(), ConstraintWeights_0=1) {
    alphaBoolBlend {blendOption=Linear}
    boneToModify {boneName=lowerarm_l}
    constraintSetup [{targetBone {boneName=dfrm_arm_spline_09_l}}]
    constraintWeights [1]
   }
   Output Pose(Result=Component To Local(ComponentPose=@n7)) {layerGroup=None}
   Save cached pose 'TargetPose'(Pose=Copy Pose From Mesh() {bUseAttachedParent=true}) {cacheName=TargetPose}
  ]}
 Check for Weapons {nodes [
   Check for Weapons() {is=title K2Node_FunctionEntry}
   > n1 = Sequence()
   n2 = Get As Fort Player Pawn() -> (As Fort Player Pawn:Fort Player Pawn Object Reference)
   > Is Valid(InputObject=@n2)
   n3 = "To String (Object)"!(InObj:Object Reference=GetCurrentWeapon!(self:Fort Pawn Object Reference=@n2) -> (ReturnValue:Fort Weapon Object Reference)) -> (ReturnValue:String)
   > Set Has Custom Weapon(Has Custom Weapon:bool=OR Boolean!(
     A=Contains(SearchIn=@n3, Substring=B_CopyRaven_GrimStrutAscended)
     B=Contains(SearchIn=@n3, Substring=B_MistRaven_CombinationAscended)
     C:bool=Contains(SearchIn=@n3, Substring=B_MistRaven_GrimStrutAscended)
    ) -> (ReturnValue)) -> (Output_Get:bool)
   Return Node!(execute=@n1.then_1, Has Weapon:bool=Get Has Custom Weapon() -> (Has Custom Weapon:bool))
  ]}
 EventGraph {nodes [
   Event BlueprintUpdateAnimation()
   > n1 = Sequence!(execute) -> (then_0, then_1, then_2:Exec, then_3:Exec, then_4:Exec, then_5:Exec, then_6:Exec, then_7:Exec)
   n2 = Check for Weapons!(execute=@n1.then_4, self) -> (then, Has Weapon:bool) {is=title K2Node_CallFunction}
   > Set arms_1(arms_1:double="To Float (Boolean)"(InBool=@n2)) -> (Output_Get:double)
   n3 = Branch(execute=@n1.then_0, Condition=Get bIsCrouching())
   > Set HeadRotYawPCT(HeadRotYawPCT:double=0.8) -> (Output_Get:double)
   Set HeadRotYawPCT(execute=@n3.else, HeadRotYawPCT=0.6)
   n4 = Branch(execute=@n1.then_1, Condition=Get bIsRightArmHidden() -> (bIsRightArmHidden:bool))
   > Set arms(arms:double=1) -> (Output_Get:double)
   Set arms(execute=@n4.else, arms=0)
   n5 = Branch(execute=@n1.then_2, Condition=Get bIsLeftArmHidden() -> (bIsLeftArmHidden:bool))
   > Set arms_0(arms_0:double=1) -> (Output_Get:double)
   Set arms_0(execute=@n5.else, arms_0=0)
   Branch(execute=@n1.then_7, Condition=Get bIsInFrontEnd())
   > Set CustomFEIdle(CustomFEIdle:double=GetCurveValue(self=GetOwnerFrontEndAnimInstance(), CurveName=Frontend_Idle_Bypass)) -> (Output_Get:double)
   n6 = Get bIsSkyDiving()
   n7 = Get bIsSkydiveFloating()
   n8 = Get bIsPlayingEmote()
   n9 = Get bIsSkydiveDiveMode()
   > Set isFrontEndNoEmote(isFrontEndNoEmote:bool=NOT Boolean(A=OR Boolean!(A=@n9, B=@n8, C=@n7, D:bool=@n6) -> (ReturnValue))) -> (Output_Get:bool)
   n10 = Branch(
    execute=@n1.then_6
    Condition=PickaxeItemDefinitionHasAnyMetaTag(Tags=Make Array("[0]"=(TagName="Cosmetics.MetaTags.MindPinch")), bRequiresEquipped=true)
   )
   > Set isCustomPickaxe(isCustomPickaxe:bool=true) -> (Output_Get:bool)
   Set isCustomPickaxe(execute=@n10.else, isCustomPickaxe=false)
   n11 = GetOwnerSkeletalMeshComponent()
   n12 = BreakRotator(InRot=BreakTransform(InTransform=GetDeltaTransformFromRefPose(self=@n11, BoneName=head, BaseName=spine_05)).Rotation)
   n13 = Get HeadRotYawPCT() -> (HeadRotYawPCT:double)
   Set HeadRot(
    execute=@n1.then_5
    HeadRot:Rotator=MakeRotator(
     Roll=float * float(A=@n12, B=0.75) {is=title K2Node_PromotableOperator}
     Pitch=float * float(A=@n12.Pitch, B=0.5)
     Yaw=float * float(A=@n12.Yaw, B=@n13)
    )
   ) -> (Output_Get:Rotator)
   n14 = NOT Boolean(A=Get bIsDBNOCarried())
   n15 = Get bIsDBNO()
   Set spr_upper_l_DBNO(execute=@n1.then_3, spr_upper_l_DBNO:double="To Float (Boolean)"(InBool=AND Boolean(A=@n15, B=@n14))) -> (Output_Get:double)
   Event OnBeginTargeting() {is=title K2Node_Event}
   > n16 = Cast To FortWeaponRanged!(execute, Object=Get CurrentWeapon() -> (CurrentWeapon:Fort Weapon Object Reference)) -> (then, CastFailed, AsFort Weapon Ranged:Fort Weapon Ranged Object Reference) {is=title K2Node_DynamicCast}
   > n17 = Branch(Condition=IsUsingScopeOrFirstPersonCamera!(self:Fort Weapon Ranged Object Reference=@n16) -> (ReturnValue:bool))
   > Set MoveThumbInScoped(MoveThumbInScoped:double=1) -> (Output_Get:double)
   Set MoveThumbInScoped(execute=@n17.else, MoveThumbInScoped=0)
   Event OnEndTargeting() {is=title K2Node_Event}
   > Set MoveThumbInScoped(MoveThumbInScoped=10)
   n18 = Event OnCharacterCustomizationLoaded()
   > n19 = Cast To FortPlayerPawn(Object=TryGetPawnOwner())
   > Set As Fort Player Pawn(As Fort Player Pawn:Fort Player Pawn Object Reference=@n19) -> (Output_Get:Fort Player Pawn Object Reference)
   > Branch(Condition=CheckForTail!(self:Custom Character Part Anim Instance Object Reference) -> (bHasTail:bool))
   > n20 = Check for Weapons()
   n21 = CheckForBackbling()
   > n22 = Branch(Condition=OR Boolean!(A=@n20, B=OR Boolean(A=@n21.bIsBackblingACape, B=@n21.bIsBackblingLong)) -> (ReturnValue))
   n23 = Get Array Item with EFort Custom Part Type(Array=@n18.CustomMeshParts, PartType=MiscOrTail)
   > SetVisibility!(execute, self:Scene Component Object Reference=@n23, bNewVisibility:bool=false, bPropagateToChildren:bool=false) -> (then)
   > Set Cape and Tail Combo(Cape and Tail Combo:bool=false) -> (Output_Get:bool)
   SetVisibility(execute=@n22.else, self=@n23, bNewVisibility=true, bPropagateToChildren=false)
   > Set Cape and Tail Combo(Cape and Tail Combo=true)
  ]}
]
variables [
 arms {type=number}
 arms_0 {type=number}
 arms_1 {type=number}
 As Fort Player Pawn {type=FortPlayerPawn}
 Cape and Tail Combo {type=boolean}
 CustomFEIdle {type=number}
 HeadRot {type=Rotator}
 HeadRotYawPCT {type=number}
 isCustomPickaxe {type=boolean}
 isFrontEndNoEmote {type=boolean}
 MoveThumbInScoped {type=number}
 spr_upper_l_DBNO {type=number}
]
