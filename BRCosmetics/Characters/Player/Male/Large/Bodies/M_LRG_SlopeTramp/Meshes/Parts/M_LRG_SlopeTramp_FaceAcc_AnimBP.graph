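# Animation Blueprint for the face accessory part of the large male SlopeTramp character body, driving flipbook brows and eyes on its overlay material.
# The anim graph copies pose and curves from the attached parent mesh.
# When character customization loads, it gets or creates a dynamic overlay material for the part mesh and stores it.
# Each animation update, after a zero-length delay, it writes left and right brow and eye curve values into that material's flipbook index parameters.
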AnimBlueprint /BRCosmetics/Characters/Player/Male/Large/Bodies/M_LRG_SlopeTramp/Meshes/Parts/M_LRG_SlopeTramp_FaceAcc_AnimBP
target_skeleton=&/M_LRG_SlopeTramp_FaceAcc_Skeleton
graphs [
 AnimGraph {nodes [Output Pose(Result=Copy Pose From Mesh() {bCopyCurves=true, bUseAttachedParent=true}) {layerGroup=None}]}
 EventGraph {nodes [
   Event BlueprintUpdateAnimation()
   > Delay(Duration=0)
   > SetScalarParameterValue(
    self=Get OverlayMat() -> (OverlayMat:Material Instance Dynamic Object Reference)
    ParameterName=flipbook_face_L_brow_index
    Value=GetCurveValue(CurveName=flipbook_face_L_brow_index)
   )
   > SetScalarParameterValue(self=Get OverlayMat(), ParameterName=flipbook_face_R_brow_index, Value=GetCurveValue(CurveName=flipbook_face_R_brow_index))
   > SetScalarParameterValue(self=Get OverlayMat(), ParameterName=flipbook_face_L_eye_index, Value=GetCurveValue(CurveName=flipbook_face_L_eye_index))
   > SetScalarParameterValue(self=Get OverlayMat(), ParameterName=flipbook_face_R_eye_index, Value=GetCurveValue(CurveName=flipbook_face_R_eye_index))
   Event OnCharacterCustomizationLoaded()
   > n1 = GetOrCreateDynamicOverlayMaterial(MeshComponent=GetPartSkeletalMeshComponent(), MaterialName=None)
   > Set OverlayMat(OverlayMat:Material Instance Dynamic Object Reference=@n1) -> (Output_Get:Material Instance Dynamic Object Reference)
  ]}
]
variables [OverlayMat {type=MaterialInstanceDynamic}]
